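# Physics asset for the RetroWheels backpack cosmetic, covering the pack, a dangling charm and a zipper.
# Four capsule bodies approximate the bones, with one large capsule for the pack and small ones for the dynamic parts.
# The two charm bodies form a chain hanging off the pack with wide swing and twist limits.
# The zipper attaches directly to the pack with tighter swing and very little twist.
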PhysicsAsset /BRCosmetics/Accessories/FORT_Backpacks/Backpack_F_MED_RetroWheels/Meshes/F_MED_RetroWheels_Pack_Physics
bodies {
 dyn_charm_1 [{shapeType=Capsule, length=0.5, radius=0.25, shapeName=dyn_charm_1_capsule, center=<0.5 0 0>, rotation=<rot -90 0 0>}]
 dyn_charm_2 [{shapeType=Capsule, length=0.5, radius=0.25, shapeName=dyn_charm_2_capsule, center=<0.5 0 0>, rotation=<rot -90 0 0>}]
 dyn_zipper [{shapeType=Capsule, length=1, radius=1, shapeName=dyn_charm_2_capsule, center=<3 0 0>, rotation=<rot -90 0 0>}]
 pack [{shapeType=Capsule, length=9, radius=9, center=<-3.6332 -2.1188 -5.4876>, rotation=<rot -90 90 -90.0003>}]
}
constraints [
 {bone1Name=dyn_charm_1, bone2Name=pack, swing1LimitDegrees=90, swing2LimitDegrees=90, twistLimitDegrees=45}
 {bone1Name=dyn_charm_2, bone2Name=dyn_charm_1, swing1LimitDegrees=90, swing2LimitDegrees=90, twistLimitDegrees=45}
 {bone1Name=dyn_zipper, bone2Name=pack, swing1LimitDegrees=55, swing2LimitDegrees=55, twistLimitDegrees=5}
]
